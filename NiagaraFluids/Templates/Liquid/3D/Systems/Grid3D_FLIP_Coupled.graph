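# Niagara template system for a GPU 3D FLIP liquid simulation coupled with a floating surfboard particle.
# Surfboard emitter burst-spawns above the origin and reads pressure, face and boundary grids from the FLIP fluid-control emitter through Grid3D readers.
# Buoyancy from those grids combines with gravity, collision and drag, while previous position and orientation are stored each update.
# User parameters expose grid size, resolution, water height, pressure iterations and collision sources.
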NiagaraSystem /NiagaraFluids/Templates/Liquid/3D/Systems/Grid3D_FLIP_Coupled
dependencies=null
emitters {Surfboard_0 {
  simTarget=GPUComputeSim
  emitterSpawnScript [
   /SetVariables(Emitter.PressureReader=DataInterface() {propertyValues {
      CellSize=1
      ClearBeforeNonIterationStage=true
      DIName=PressureGrid
      EmitterName=Grid3D_FLIP_FluidControl_Emitter
      NumAttributes=0
      NumCellsMaxAxis=10
      OverrideBufferFormat=Float
      PreviewAttribute=None
      SetResolutionMethod=Independent
      bOverrideFormat=false
      bPreviewGrid=false
      NumCells=<3 3 3>
      RenderTargetUserParameter {parameter {name=None, typeDefHandle {registeredTypeIndex=4}, varData []}}
      WorldBBoxSize=<100 100 100>
     }}) {isSetParameters=true}
   /SetVariables(Emitter.FaceReader=DataInterface() {propertyValues {
      CellSize=1
      ClearBeforeNonIterationStage=true
      DIName=FaceGrid
      EmitterName=Grid3D_FLIP_FluidControl_Emitter
      NumAttributes=0
      NumCellsMaxAxis=10
      OverrideBufferFormat=Float
      PreviewAttribute=None
      SetResolutionMethod=Independent
      bOverrideFormat=false
      bPreviewGrid=false
      NumCells=<3 3 3>
      RenderTargetUserParameter {parameter {name=None, typeDefHandle {registeredTypeIndex=4}, varData []}}
      WorldBBoxSize=<100 100 100>
     }}) {isSetParameters=true}
   /SetVariables(Emitter.BoundaryReader=DataInterface() {propertyValues {
      CellSize=1
      ClearBeforeNonIterationStage=true
      DIName=TransientGrid
      EmitterName=Grid3D_FLIP_FluidControl_Emitter
      NumAttributes=0
      NumCellsMaxAxis=10
      OverrideBufferFormat=Float
      PreviewAttribute=None
      SetResolutionMethod=Independent
      bOverrideFormat=false
      bPreviewGrid=false
      NumCells=<3 3 3>
      RenderTargetUserParameter {parameter {name=None, typeDefHandle {registeredTypeIndex=4}, varData []}}
      WorldBBoxSize=<100 100 100>
     }}) {isSetParameters=true}
  ]
  emitterUpdateScript [
   &EmitterState(Life Cycle Mode=::NewEnumerator1, Loop Behavior=::NewEnumerator1, Loop Duration Mode=::NewEnumerator1, UseLoopDelay=false)
   &SpawnBurst_Instantaneous(Use Spawn Probability=false, UseLoopCountLimit=false)
   &self:Surfboard_1.NiagaraScratchPadContainer_0.ScratchModule_06_1(Reader:NiagaraDataInterfaceGrid3DCollectionReader=$Emitter.BoundaryReader, WorldGridExtents:Vector3f=$System.WorldGridExtents) {module=ScratchModule_06_6}
  ]
  particleSpawnScript [
   &self:Surfboard_1.NiagaraScratchPadContainer_0.ScratchModule_04_2(Density=0.5, Radius:Vector3f=<40 20 20>) {module=ScratchModule_04}
   /SetVariables(
    Particles.MeshOrientation=&MakeQuatFromAxisAngle(Axis=<0 1 0>, AngleInDegrees=45)
    Particles.Position=<0 0 200>
   ) {isSetParameters=true}
  ]
  particleUpdateScript [
   &ParticleState(Kill Particles When Lifetime Has Elapsed=false)
   &GravityForce()
   /NiagaraFluids/Modules/Grid3D/Grid3D_ExternalObjectBuoyancyForce(
    BoundaryReader:NiagaraDataInterfaceGrid3DCollectionReader=$Emitter.BoundaryReader
    DebugDraw=DataInterface() {propertyValues {OverrideMaxLineInstances=2133}}
    FaceReader:NiagaraDataInterfaceGrid3DCollectionReader=$Emitter.FaceReader
    PressureReader:NiagaraDataInterfaceGrid3DCollectionReader=$Emitter.PressureReader
    UnitToWorld:NiagaraMatrix=$System.Grid3D_CreateUnitToWorldTransform.UnitToWorld
    WorldToUnit:NiagaraMatrix=$System.Grid3D_CreateUnitToWorldTransform.WorldToUnit
    dx:NiagaraFloat=$Emitter.dx
   )
   &Collision(
    Age Colliding Particles=false
    Friction=0
    GPU Collision Type=::NewEnumerator2
    KillOnCollision=false
    Module.Control Rest State Via Surface Normal=false
    Module.EnableMaxCollisionCount=false
    Module.Randomize Collision Normal Vector=false
    Restitution=0
    Simple Friction=false
    Static Friction=0
    Static Friction Engagement Speed=0
   )
   /SetVariables(Particles.PrevMeshOrientation:Quat4f=$Particles.MeshOrientation, Particles.PrevPosition:NiagaraPosition=$Particles.Position) {isSetParameters=true}
   &Drag(Drag=0.5, Rotational Drag=0, Use Linear Drag=true, Use Rotational Drag=false)
   &SolveForcesAndVelocity(Clamp Velocity=false, Limit Acceleration=false) {module=SolveForcesAndVelocity001}
   &self:Surfboard_1.NiagaraScratchPadContainer_0.ScratchModule_3(DeltaTime:NiagaraFloat=$Engine.DeltaTime, Drag=0.5) {module=ScratchModule}
  ]
  rendererClasses [&NiagaraSpriteRendererProperties, &NiagaraMeshRendererProperties]
  renderers [
   {rendererClass=&NiagaraSpriteRendererProperties}
   {rendererClass=&NiagaraMeshRendererProperties, rendererIndex=1}
  ]
 }}
summary {
 systemName=Grid3D_FLIP_Coupled
 emitters [
  {
   emitterName=Grid3D_FLIP_FluidControl_Emitter
   simTarget=GPUComputeSim
   rendererClasses [&NiagaraSpriteRendererProperties, &NiagaraMeshRendererProperties]
  }
  {
   emitterName=Surfboard_0
   simTarget=GPUComputeSim
   rendererClasses [&NiagaraSpriteRendererProperties, &NiagaraMeshRendererProperties]
  }
 ]
 userVariables [
  User.Collide_Meshes {
   type=NiagaraDataInterfaceRigidMeshCollisionQuery
   description=""
   defaultValue {
    struct=NiagaraExt_VariableValue_DataInterface
    value {
     dataInterface=&self:NiagaraDataInterfaceRigidMeshCollisionQuery_1
     dataInterfaceClass=/Script/Niagara.NiagaraDataInterfaceRigidMeshCollisionQuery
    }
   }
  }
  User.Collision Velocity Mult {type=NiagaraFloat, defaultValue=1, description=""}
  User.GeometryCollectionCollisions {
   type=NiagaraDataInterfaceGeometryCollection
   description=""
   defaultValue {
    struct=NiagaraExt_VariableValue_DataInterface
    value {
     dataInterface=&self:NiagaraDataInterfaceGeometryCollection_0
     dataInterfaceClass=/Script/ChaosNiagara.NiagaraDataInterfaceGeometryCollection
    }
   }
  }
  User.LandscapeCollisions {
   type=NiagaraDataInterfaceLandscape
   description=""
   defaultValue {
    struct=NiagaraExt_VariableValue_DataInterface
    value {
     dataInterface=&self:NiagaraDataInterfaceLandscape_0
     dataInterfaceClass=&NiagaraDataInterfaceLandscape
    }
   }
  }
  User.Num Cells Max Axis {type=NiagaraInt32, defaultValue=92}
  User.Particles Per Cell {type=NiagaraInt32, defaultValue=6}
  User.PhysicsCollisions {
   type=NiagaraDataInterfacePhysicsAsset
   description=""
   defaultValue {
    struct=NiagaraExt_VariableValue_DataInterface
    value {
     dataInterface=&self:NiagaraDataInterfacePhysicsAsset_0
     dataInterfaceClass=/Script/Niagara.NiagaraDataInterfacePhysicsAsset
    }
   }
  }
  User.Pressure Iterations {type=NiagaraInt32, defaultValue=200}
  User.Show Bounds {type=NiagaraBool, defaultValue=true}
  User.Water Height {type=NiagaraFloat, defaultValue=100, description=""}
  User.World Grid Extents {type=Vector3f, description="", defaultValue {struct=Vector3f, value=<300 300 150>}}
 ]
}
