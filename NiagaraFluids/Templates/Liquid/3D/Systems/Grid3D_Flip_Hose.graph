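# Niagara Fluids template system for a 3D FLIP liquid simulation of a hose, run on the GPU.
# A fluid control emitter runs the simulation with sprite and mesh renderers, alongside a secondary sprite emitter that is disabled.
# User parameters set grid extents, resolution, particles per cell, pressure iterations and water height, plus collision inputs for rigid meshes, geometry collections, landscape and physics assets.
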NiagaraSystem /NiagaraFluids/Templates/Liquid/3D/Systems/Grid3D_Flip_Hose.Grid3D_FLIP_Hose
dependencies=null
summary {
 systemName=Grid3D_FLIP_Hose
 emitters [
  {
   emitterName=Grid3D_FLIP_FluidControl_Emitter
   simTarget=GPUComputeSim
   rendererClasses [&NiagaraSpriteRendererProperties, &NiagaraMeshRendererProperties]
  }
  {
   bEnabled=false
   emitterName=Grid3D_FLIP_Secondary_Emitter
   simTarget=GPUComputeSim
   rendererClasses [&NiagaraSpriteRendererProperties]
  }
 ]
 userVariables [
  User.Collide_Meshes {
   type=NiagaraDataInterfaceRigidMeshCollisionQuery
   description=""
   defaultValue {
    struct=NiagaraExt_VariableValue_DataInterface
    value {
     dataInterface=&self:NiagaraDataInterfaceRigidMeshCollisionQuery_1
     dataInterfaceClass=/Script/Niagara.NiagaraDataInterfaceRigidMeshCollisionQuery
    }
   }
  }
  User.Collision Velocity Mult {type=NiagaraFloat, defaultValue=1, description=""}
  User.GeometryCollectionCollisions {
   type=NiagaraDataInterfaceGeometryCollection
   description=""
   defaultValue {
    struct=NiagaraExt_VariableValue_DataInterface
    value {
     dataInterface=&self:NiagaraDataInterfaceGeometryCollection_0
     dataInterfaceClass=/Script/ChaosNiagara.NiagaraDataInterfaceGeometryCollection
    }
   }
  }
  User.LandscapeCollisions {
   type=NiagaraDataInterfaceLandscape
   description=""
   defaultValue {
    struct=NiagaraExt_VariableValue_DataInterface
    value {
     dataInterface=&self:NiagaraDataInterfaceLandscape_0
     dataInterfaceClass=&NiagaraDataInterfaceLandscape
    }
   }
  }
  User.Num Cells Max Axis {type=NiagaraInt32, defaultValue=92}
  User.Particles Per Cell {type=NiagaraInt32, defaultValue=8}
  User.PhysicsCollisions {
   type=NiagaraDataInterfacePhysicsAsset
   description=""
   defaultValue {
    struct=NiagaraExt_VariableValue_DataInterface
    value {
     dataInterface=&self:NiagaraDataInterfacePhysicsAsset_0
     dataInterfaceClass=/Script/Niagara.NiagaraDataInterfacePhysicsAsset
    }
   }
  }
  User.Pressure Iterations {type=NiagaraInt32, defaultValue=200}
  User.Show Bounds {type=NiagaraBool}
  User.Water Height {type=NiagaraFloat, defaultValue=150, description=""}
  User.World Grid Extents {type=Vector3f, description="", defaultValue {struct=Vector3f, value=<400 400 500>}}
 ]
}
