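# Niagara template system for a 3D FLIP liquid pool on the GPU, with user parameters for water height, resolution, extents, pressure iterations and collision sources.
# The FluidControl emitter, listed only in summary, supplies simulation, boundary and SDF grids to the Secondary emitter.
# That emitter spawns spray, foam and bubble sprites via an emission-position grid and counter, using fluid velocity.
# Particles age by state, pass state to materials and die inside boundaries.
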NiagaraSystem /NiagaraFluids/Templates/Liquid/3D/Systems/Grid3D_Flip_Pool.Grid3D_FLIP_Pool
dependencies=null
emitters {Grid3D_FLIP_Secondary_Emitter {
  simTarget=GPUComputeSim
  emitterSpawnScript [
   /SetVariables(Emitter.SimGridReader=DataInterface() {propertyValues {
      CellSize=1
      ClearBeforeNonIterationStage=true
      DIName=SimGrid
      EmitterName=Grid3D_FLIP_FluidControl_Emitter
      NumAttributes=0
      NumCellsMaxAxis=10
      OverrideBufferFormat=Float
      PreviewAttribute=None
      SetResolutionMethod=Independent
      bOverrideFormat=false
      bPreviewGrid=false
      NumCells=<3 3 3>
      RenderTargetUserParameter {parameter {name=None, typeDefHandle {registeredTypeIndex=4}, varData []}}
      WorldBBoxSize=<100 100 100>
     }}) {isSetParameters=true}
   /SetVariables(Emitter.BoundaryReader=DataInterface() {propertyValues {
      CellSize=1
      ClearBeforeNonIterationStage=true
      DIName=TransientGrid
      EmitterName=Grid3D_FLIP_FluidControl_Emitter
      NumAttributes=0
      NumCellsMaxAxis=10
      OverrideBufferFormat=Float
      PreviewAttribute=None
      SetResolutionMethod=Independent
      bOverrideFormat=false
      bPreviewGrid=false
      NumCells=<3 3 3>
      RenderTargetUserParameter {parameter {name=None, typeDefHandle {registeredTypeIndex=4}, varData []}}
      WorldBBoxSize=<100 100 100>
     }}) {isSetParameters=true}
   /SetVariables(Emitter.SDFReader=DataInterface() {propertyValues {
      CellSize=1
      ClearBeforeNonIterationStage=true
      DIName=SDFGrid
      EmitterName=Grid3D_FLIP_FluidControl_Emitter
      NumAttributes=0
      NumCellsMaxAxis=10
      OverrideBufferFormat=Float
      PreviewAttribute=None
      SetResolutionMethod=Independent
      bOverrideFormat=false
      bPreviewGrid=false
      NumCells=<3 3 3>
      RenderTargetUserParameter {parameter {name=None, typeDefHandle {registeredTypeIndex=4}, varData []}}
      WorldBBoxSize=<100 100 100>
     }}) {isSetParameters=true}
   /SetVariables(Emitter.MaxSecondaryParticlesPerFrame=15000) {isSetParameters=true}
   /SetVariables(Emitter.EmissionPositions=DataInterface() {propertyValues {
      ClearBeforeNonIterationStage=true
      NumAttributes=3
      NumCellsMaxAxis=3
      NumCellsX=1
      NumCellsY=1
      OverrideBufferFormat=Float
      PreviewAttribute=None
      SetGridFromMaxAxis=false
      bOverrideFormat=false
      bPreviewGrid=false
      RenderTargetUserParameter {parameter {name=None, typeDefHandle {registeredTypeIndex=4}, varData []}}
      WorldBBoxSize=<100 100>
     }}) {isSetParameters=true}
   /NiagaraFluids/Modules/Grid2D/Grid2D_SetResolution(
    Grid:NiagaraDataInterfaceGrid2DCollection=$Emitter.EmissionPositions
    NumCellsX:NiagaraInt32=$Emitter.MaxSecondaryParticlesPerFrame
    NumCellsY=1
   )
   /SetVariables(Emitter.EmissionCounter=DataInterface() {propertyValues {
      CellSize=1
      ClearBeforeNonIterationStage=true
      MaxNeighborsPerCell=1
      NumCellsMaxAxis=10
      SetResolutionMethod=Independent
      NumCells=<2 2 2>
      WorldBBoxSize=<100 100 100>
     }}) {isSetParameters=true}
   &self:Grid3D_FLIP_Secondary_Emitter_0.NiagaraScratchPadContainer_2.CreateGridAttributeIndices_2(SimGridReader:NiagaraDataInterfaceGrid3DCollectionReader=$Emitter.SimGridReader) {module=CreateGridAttributeIndices}
   &self:Grid3D_FLIP_Secondary_Emitter_0.NiagaraScratchPadContainer_2.ScratchModule_01_5() {module=ScratchModule_01}
  ]
  emitterUpdateScript [
   &EmitterState()
   &self:Grid3D_FLIP_Secondary_Emitter_0.NiagaraScratchPadContainer_2.ScratchModule_06_1(Reader:NiagaraDataInterfaceGrid3DCollectionReader=$Emitter.SimGridReader, WorldGridExtents:Vector3f=$System.WorldGridExtents) {module=ScratchModule_06}
   /SetVariables(
    Emitter.GridCenterPosition:Vector3f=$System.Grid3D_CreateUnitToWorldTransform.GridCenterPosition
    Emitter.GridOrientation:Quat4f=$System.Grid3D_CreateUnitToWorldTransform.GridOrientation
    Emitter.LocalToWorld:NiagaraMatrix=$System.Grid3D_CreateUnitToWorldTransform.LocalToWorld
    Emitter.UnitToWorld:NiagaraMatrix=$System.Grid3D_CreateUnitToWorldTransform.UnitToWorld
    Emitter.WorldToLocal:NiagaraMatrix=$System.Grid3D_CreateUnitToWorldTransform.WorldToLocal
    Emitter.WorldToLocal0:Vector4f=$System.Grid3D_CreateUnitToWorldTransform.WorldToLocal0
    Emitter.WorldToLocal1:Vector4f=$System.Grid3D_CreateUnitToWorldTransform.WorldToLocal1
    Emitter.WorldToLocal2:Vector4f=$System.Grid3D_CreateUnitToWorldTransform.WorldToLocal2
    Emitter.WorldToLocal3:Vector4f=$System.Grid3D_CreateUnitToWorldTransform.WorldToLocal3
    Emitter.WorldToUnit:NiagaraMatrix=$System.Grid3D_CreateUnitToWorldTransform.WorldToUnit
   ) {isSetParameters=true}
   /SetVariables(Emitter.DeltaTime:NiagaraFloat=$Engine.DeltaTime) {isSetParameters=true}
   &SpawnPerFrame(Spawn Count:NiagaraInt32=$Emitter.MaxSecondaryParticlesPerFrame, Use Spawn Probability=false)
  ]
  particleSpawnScript [
   &self:Grid3D_FLIP_Secondary_Emitter_0.NiagaraScratchPadContainer_2.ScratchModule_4(
    Lifetime:NiagaraFloat=&RandomRangeFloat(Minimum=2, Maximum=3)
    SpriteSize:Vector2f=&Vector2DFromFloat(Value=&FloatFromCurve(
      FloatCurve=DataInterface() {propertyValues {Curve {keys [<at 0 5 RCIM_Cubic>,<at 0.9 8>]}}}
      CurveIndex=&RandomRangeFloat(Minimum=0, Maximum=1)
      Scale Curve=1.5
     ))
   ) {module=ScratchModule}
   /NiagaraFluids/Modules/Grid3D/Secondary/Grid3D_Secondary_SpawnParticles(
    EmissionCounter:NiagaraDataInterfaceNeighborGrid3D=$Emitter.EmissionCounter
    EmissionPositions:NiagaraDataInterfaceGrid2DCollection=$Emitter.EmissionPositions
    FluidVelocityFraction=1
    MaxSecondaryParticlesPerFrame:NiagaraInt32=$Emitter.MaxSecondaryParticlesPerFrame
    MaxSpeed=1000
    SimDt:NiagaraFloat=$Emitter.DeltaTime
    SimGridReader:NiagaraDataInterfaceGrid3DCollectionReader=$Emitter.SimGridReader
    UnitToWorld:NiagaraMatrix=$Emitter.UnitToWorld
    VelocityMult=1
    VelocityOffset=1
    dx:NiagaraFloat=$Emitter.dx
   )
  ]
  particleUpdateScript [
   /NiagaraFluids/Modules/Grid3D/Secondary/Grid3D_Secondary_UpdateParticles(
    BoundaryReader:NiagaraDataInterfaceGrid3DCollectionReader=$Emitter.BoundaryReader
    BubbleAgingRate=1
    BubbleBounancy:Vector3f=<0 0 980>
    FoamAgingRate=1
    SDFReader:NiagaraDataInterfaceGrid3DCollectionReader=$Emitter.SDFReader
    SimDt:NiagaraFloat=$Emitter.DeltaTime
    SimGridReader:NiagaraDataInterfaceGrid3DCollectionReader=$Emitter.SimGridReader
    SprayAgingRate=1
    WorldToUnit:NiagaraMatrix=$Emitter.WorldToUnit
   )
   &DynamicMaterialParameters(
    Index 0 Param 1=&RandomRangeFloat(Minimum=0, Maximum=1)
    Index 0 Param 2=&RandomRangeFloat(Minimum=0, Maximum=1)
    Index 0 Param 4=&MakeFloatFromInt(Int=$Particles.State)
    Module.Param0WriteEnabled=true
    Module.Param1WriteEnabled=true
    Module.Param2WriteEnabled=false
    Module.Param3WriteEnabled=true
   )
   &ParticleState(DeltaTime:NiagaraFloat=&Multiply_Float(A=$Emitter.DeltaTime, B=$Transient.AgingRate))
   &KillParticles(Kill Particles=$Output.Grid3D_Secondary_UpdateParticles.IsInsideBoundary)
   &SubUVAnimation(SubUV Lookup Index:NiagaraFloat=&Frac_Float(Float=$Particles.Age)) {enabled=false}
  ]
  rendererClasses [&NiagaraSpriteRendererProperties]
  renderers [
   {rendererClass=&NiagaraSpriteRendererProperties}
   {rendererClass=&NiagaraSpriteRendererProperties, rendererIndex=1}
  ]
 }}
summary {
 systemName=Grid3D_FLIP_Pool
 emitters [
  {
   emitterName=Grid3D_FLIP_FluidControl_Emitter
   simTarget=GPUComputeSim
   rendererClasses [&NiagaraSpriteRendererProperties, &NiagaraMeshRendererProperties]
  }
  {emitterName=Grid3D_FLIP_Secondary_Emitter, simTarget=GPUComputeSim, rendererClasses [&NiagaraSpriteRendererProperties]}
 ]
 userVariables [
  User.Collide_Meshes {
   type=NiagaraDataInterfaceRigidMeshCollisionQuery
   description=""
   defaultValue {
    struct=NiagaraExt_VariableValue_DataInterface
    value {
     dataInterface=&self:NiagaraDataInterfaceRigidMeshCollisionQuery_1
     dataInterfaceClass=/Script/Niagara.NiagaraDataInterfaceRigidMeshCollisionQuery
    }
   }
  }
  User.Collision Velocity Mult {type=NiagaraFloat, defaultValue=1, description=""}
  User.GeometryCollectionCollisions {
   type=NiagaraDataInterfaceGeometryCollection
   description=""
   defaultValue {
    struct=NiagaraExt_VariableValue_DataInterface
    value {
     dataInterface=&self:NiagaraDataInterfaceGeometryCollection_0
     dataInterfaceClass=/Script/ChaosNiagara.NiagaraDataInterfaceGeometryCollection
    }
   }
  }
  User.LandscapeCollisions {
   type=NiagaraDataInterfaceLandscape
   description=""
   defaultValue {
    struct=NiagaraExt_VariableValue_DataInterface
    value {
     dataInterface=&self:NiagaraDataInterfaceLandscape_0
     dataInterfaceClass=&NiagaraDataInterfaceLandscape
    }
   }
  }
  User.Num Cells Max Axis {type=NiagaraInt32, defaultValue=64}
  User.Particles Per Cell {type=NiagaraInt32, defaultValue=8}
  User.PhysicsCollisions {
   type=NiagaraDataInterfacePhysicsAsset
   description=""
   defaultValue {
    struct=NiagaraExt_VariableValue_DataInterface
    value {
     dataInterface=&self:NiagaraDataInterfacePhysicsAsset_0
     dataInterfaceClass=/Script/Niagara.NiagaraDataInterfacePhysicsAsset
    }
   }
  }
  User.Pressure Iterations {type=NiagaraInt32, defaultValue=150}
  User.Show Bounds {type=NiagaraBool}
  User.Water Height {type=NiagaraFloat, defaultValue=150, description=""}
  User.World Grid Extents {type=Vector3f, description="", defaultValue {struct=Vector3f, value=<800 800 400>}}
 ]
}
