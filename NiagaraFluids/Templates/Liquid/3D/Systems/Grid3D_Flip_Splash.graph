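# Niagara Fluids template system for a 3D FLIP liquid splash, simulated on a grid with particles.
# A single GPU compute fluid-control emitter carries sprite and mesh renderers.
# User parameters set grid extents and resolution, particles per cell, pressure iterations, water height and bounds display.
# Collision inputs cover rigid meshes, geometry collections, landscape and physics assets, with a collision velocity multiplier.
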NiagaraSystem /NiagaraFluids/Templates/Liquid/3D/Systems/Grid3D_Flip_Splash
dependencies=null
summary {
 systemName=Grid3D_Flip_Splash
 emitters [{
   emitterName=Grid3D_FLIP_FluidControl_Emitter
   simTarget=GPUComputeSim
   rendererClasses [&NiagaraSpriteRendererProperties, &NiagaraMeshRendererProperties]
  }]
 userVariables [
  User.Collide_Meshes {
   type=NiagaraDataInterfaceRigidMeshCollisionQuery
   description=""
   defaultValue {
    struct=NiagaraExt_VariableValue_DataInterface
    value {
     dataInterface=&self:NiagaraDataInterfaceRigidMeshCollisionQuery_1
     dataInterfaceClass=/Script/Niagara.NiagaraDataInterfaceRigidMeshCollisionQuery
    }
   }
  }
  User.Collision Velocity Mult {type=NiagaraFloat, defaultValue=1, description=""}
  User.GeometryCollectionCollisions {
   type=NiagaraDataInterfaceGeometryCollection
   description=""
   defaultValue {
    struct=NiagaraExt_VariableValue_DataInterface
    value {
     dataInterface=&self:NiagaraDataInterfaceGeometryCollection_0
     dataInterfaceClass=/Script/ChaosNiagara.NiagaraDataInterfaceGeometryCollection
    }
   }
  }
  User.LandscapeCollisions {
   type=NiagaraDataInterfaceLandscape
   description=""
   defaultValue {
    struct=NiagaraExt_VariableValue_DataInterface
    value {
     dataInterface=&self:NiagaraDataInterfaceLandscape_0
     dataInterfaceClass=&NiagaraDataInterfaceLandscape
    }
   }
  }
  User.Num Cells Max Axis {type=NiagaraInt32, defaultValue=128}
  User.Particles Per Cell {type=NiagaraInt32, defaultValue=4}
  User.PhysicsCollisions {
   type=NiagaraDataInterfacePhysicsAsset
   description=""
   defaultValue {
    struct=NiagaraExt_VariableValue_DataInterface
    value {
     dataInterface=&self:NiagaraDataInterfacePhysicsAsset_0
     dataInterfaceClass=/Script/Niagara.NiagaraDataInterfacePhysicsAsset
    }
   }
  }
  User.Pressure Iterations {type=NiagaraInt32, defaultValue=59}
  User.Show Bounds {type=NiagaraBool, defaultValue=true}
  User.Water Height {type=NiagaraFloat, defaultValue=0, description=""}
  User.World Grid Extents {type=Vector3f, description="", defaultValue {struct=Vector3f, value=<400 400 200>}}
 ]
}
